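; Looping gameplay cue notify that drives the effects of Umbramolt Kipper's dash charge on the target pawn.
; On loop start, if the pawn's display name contains "WanderingBoss", it prints a debug string and swaps the first particle component to an enlarged Masamune boss Niagara system.
; After a short delay, if the physics field CVar is on, it attaches a wash-force field actor to the pawn.
; Removal destroys that force actor.
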(event EventOnLoopingStart(EitherVFX-preferthisversion)
  (params (MyTarget :type Actor%20Object%20Reference) (Parameters :type Gameplay%20Cue%20Parameters%20Structure%20%28by%20ref%29) (ParticleComponents :type Array%20of%20FXSystem%20Component%20Object%20References) (AudioComponents :type Array%20of%20Audio%20Component%20Object%20References))
  (|CallParentFunction MyTarget Parameters ParticleComponents AudioComponents)
  (Utilities|FlowControl|Sequence
    (:then_0
      (bind _asniagara_particle_system_component (Utilities|Casting|CastToNiagaraComponent (Utilities|Array|Get(acopy) ParticleComponents)))
      (bind _asfort_pawn (Utilities|Casting|CastToFortPawn MyTarget))
      (Variables|Default|SetAs%20Fort%20Pawn _asfort_pawn)
      (if (>= (Utilities|String|FindSubstring (Utilities|GetDisplayName _asfort_pawn) "WanderingBoss") 0)
        (Development|PrintString 0 "TyphoonBladeWorkOnWanderingBoss")
        (Niagara|SetNiagaraSystemAsset _asniagara_particle_system_component "/MasamuneGameplay/Assets/FX/NewFX/NS_Masamune_Asborb_WanderingBoss.NS_Masamune_Asborb_WanderingBoss" true)
        (Transformation|SetWorldScale3D _asniagara_particle_system_component "3.000000,3.000000,3.000000")))
    (:then_1
      (Utilities|FlowControl|Delay 0 0.1)
      (bind _returnvalue (Development|GetConsoleVariableIntValue "r.PhysicsField.EnableField"))
      (if (== _returnvalue 1)
        (bind _returnvalue_2 (Game|SpawnActorfromClass 0 "/Game/Effects/Niagara/Blueprints/PhsyicsFields/FN_WashForce.FN_WashForce_C" (struct :Location "0, 0, 0" :Rotation "0, 0, 0" :Scale "1.000000,1.000000,1.000000") "Undefined" "MultiplyWithRoot" 0 1e+03 0.8 -999.0 0))
        (Variables|Default|SetForce _returnvalue_2)
        (Transformation|AttachActorToActor (Variables|Default|GetForce) (Variables|Default|GetAs%20Fort%20Pawn) "None" "KeepRelative" "KeepRelative" "KeepRelative" true)))))

(event EventOnRemoval(EitherVFX-preferthisversion)
  (params (MyTarget :type Actor%20Object%20Reference) (Parameters :type Gameplay%20Cue%20Parameters%20Structure%20%28by%20ref%29) (ParticleComponents :type Array%20of%20FXSystem%20Component%20Object%20References) (AudioComponents :type Array%20of%20Audio%20Component%20Object%20References) (BurstCameraShakeInstance :type Camera%20Shake%20Base%20Object%20Reference) (BurstDecalInstance :type Decal%20Actor%20Object%20Reference))
  (Actor|DestroyActor (Variables|Default|GetForce)))

(fn ConstructionScript)
