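; Looping gameplay cue for the Umbramolt Kipper dash that adds an on-screen sound indicator when the dash starts.
; The looping-start event calls the parent implementation, then shows a gunshot-type indicator with a Creative objective attack icon at the target actor's location.
; The indicator passes the target's team with Neutral and Hostile affiliations and a limited audible range.
; It fires only at loop start, with no update or removal logic in this graph.
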(event EventOnLoopingStart(EitherVFX-preferthisversion)
  (params (MyTarget :type Actor%20Object%20Reference) (Parameters :type Gameplay%20Cue%20Parameters%20Structure%20%28by%20ref%29) (ParticleComponents :type Array%20of%20FXSystem%20Component%20Object%20References) (AudioComponents :type Array%20of%20Audio%20Component%20Object%20References))
  (|CallParentFunction MyTarget Parameters ParticleComponents AudioComponents)
  (Fortnite|ShowSoundIndicatorLocalforTeam :WorldContextObject MyTarget :Location (Transformation|GetActorLocation MyTarget) :MaxAudibleDistance 5e+03 :IndicatorType "Gunshot" :Team (Team|GetActorTeam MyTarget) :Affiliations (Utilities|Array|MakeArray "Neutral" "Hostile") :Instigator MyTarget :Tint "(R=0.000000,G=0.000000,B=0.000000,A=0.000000)" :OverrideIcon "/Game/Athena/HUD/LTM/Playground/Materials/T_UI_CreativeObjectiveAttack.T_UI_CreativeObjectiveAttack" :AttenuationMaxRangeMultiplier 1.0 :AttenuationIntensityMultiplier 1.0))

(fn ConstructionScript)
