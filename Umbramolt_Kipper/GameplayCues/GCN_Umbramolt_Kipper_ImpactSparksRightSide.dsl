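; Burst gameplay cue for Umbramolt Kipper right-side impact sparks, aligning each spawned particle component to the target pawn's aim rotation.
; With physics fields enabled by console variable, each loop pass spawns a radial force field ahead of the pawn.
; If the mesh's display name contains "WanderingBoss", it prints a debug string, swaps the first component's Niagara system, flags and restarts the second, then rescales and offsets all components.
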(fn OnBurst(EitherVFX-preferthisversion)
  (params (MyTarget :type Actor%20Object%20Reference) (Parameters :type Gameplay%20Cue%20Parameters%20Structure%20%28by%20ref%29) (ParticleComponents :type Array%20of%20FXSystem%20Component%20Object%20References) (AudioComponents :type Array%20of%20Audio%20Component%20Object%20References) (BurstCameraShakeInstance :type Camera%20Shake%20Base%20Object%20Reference) (BurstDecalInstance :type Decal%20Actor%20Object%20Reference))
  (bind _returnvalue_location (Math|Transform|BreakTransform (Transformation|GetActorTransform (LocalVariables|GetPawn))))
  (bind _output (Utilities|Array|Get(acopy) ParticleComponents 1))
  (bind _output_2 (Utilities|Array|Get(acopy) ParticleComponents))
  (Utilities|FlowControl|Sequence
    (:then_0
      (bind _asfort_pawn (Utilities|Casting|CastToFortPawn MyTarget))
      (bind _output_get (LocalVariables|SetPawn _asfort_pawn))
      (for _array_element ParticleComponents
        (Transformation|SetWorldRotation _array_element (Pawn|GetBaseAimRotation (LocalVariables|GetPawn)) false false)
        (bind _returnvalue (Development|GetConsoleVariableIntValue "r.PhysicsField.EnableField"))
        (if (== _returnvalue 1)
          (Game|SpawnActorfromClass 0 "/Game/Effects/Niagara/Blueprints/PhsyicsFields/FN_RadialForce.FN_RadialForce_C" (Math|Transform|MakeTransform (+ _returnvalue_location (* (Transformation|GetForwardVector (Utilities|Array|Get(acopy) ParticleComponents)) "1000.000000,1000.000000,1000.000000"))) "Undefined" "MultiplyWithRoot" 0 0.5 0.75 1.5e+03 0.8 0.0 0)))
      (if (>= (Utilities|String|FindSubstring (Utilities|GetDisplayName (Variables|/Script/Engine.Character|GetMesh _output_get)) "WanderingBoss") 0)
        ; This part only work on WBoss
        (Development|PrintString 0 "TyphoonBladeWorkOnWanderingBoss")
        (Components|Activation|Deactivate _output_2)
        (bind _asniagara_particle_system_component (Utilities|Casting|CastToNiagaraComponent _output_2))
        (Niagara|SetNiagaraSystemAsset _asniagara_particle_system_component "/MasamuneGameplay/Assets/FX/NewFX/NS_Masamune_OldWangderingBossSlash.NS_Masamune_OldWangderingBossSlash" false)
        (Components|Activation|Activate _asniagara_particle_system_component true)
        (Effects|Components|ParticleSystem|SetBoolParameter _output "bIsWanderingBoss" true)
        (Components|Activation|Deactivate _output)
        (Components|Activation|Activate _output true)
        (Effects|Components|ParticleSystem|SetBoolParameter _output "bIsWanderingBoss" true)
        (for _array_element_2 ParticleComponents
          (Transformation|SetWorldScale3D _array_element_2 "1.500000,1.500000,1.500000")
          (Effects|Components|ParticleSystem|SetVectorParameter _array_element_2 "FXvelocityOffset" "1200.000000,0.000000,0.000000")
          (Effects|Components|ParticleSystem|SetVectorParameter _array_element_2 "FXPositionOffset" "-300.000000,0.000000,70.000000"))
        (else
          _))
      (return))))
